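# Post-process style AnimBP for the PixieParts Magma male body cosmetic, driving arm offsets and pad physics.
# Copies the attached parent's pose into a Control Rig fed parachute, emote and front-end states plus upper-arm transforms chosen by Hat mesh name.
# The cached result feeds two differently tuned two-pad AnimDynamics spring chains, selected by front-end state.
# The Face part component, stored as Body Mesh, becomes a tick prerequisite.
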AnimBlueprint /BRCosmetics/Characters/Player/Male/Medium/Bodies/M_MED_PixieParts_Magma/Meshes/M_MED_PixieParts_Magma_AnimBP
target_skeleton=&/M_MED_PixieParts_Magma_Skeleton
graphs [
 AnimGraph {nodes [
   n1 = Anim Dynamics - Bone: base_pad_reg_r!(
    ComponentPose=Local To Component(LocalPose=Use cached pose 'CtrlRig_Cache'())
    Alpha=1
    GravityOverride:Vector="0,0,0"
    ExternalForce="0,0,0"
   ) -> (Pose) {
    angularDampingOverride=1
    angularSpringConstant=5000
    bAngularSpring=true
    bOverrideAngularDamping=true
    bUseGravityOverride=true
    gravityScale=0.35
    simulationSpace=RootRelative
    alphaBoolBlend {blendOption=Linear}
    boundBone {boneName=base_pad_reg_r}
    componentAppliedLinearAccClamp=<3000 3000 3000>
    componentLinearAccScale=<0.15 0.15 0.15>
    componentLinearVelScale=<0.15 0.15 0.15>
    gravityOverride {z=-980}
    physicsBodyDefinitions [{
      boundBone {boneName=base_pad_reg_r}
      boxExtents=<35 15 15>
      constraintSetup {angularLimitsMax=<5 5 10>, angularLimitsMin=<-5 -5 -10>, angularTarget=<1 0 0>}
      localJointOffset=<2 0 0>
     }]
    simSpaceSettings {maxAngularAcceleration=35, maxAngularVelocity=7, simSpaceAngularAlpha=0.2}
   }
   n2 = Anim Dynamics - Bone: base_pad_leg_l!(ComponentPose=@n1, Alpha=1, GravityOverride:Vector="0,0,0", ExternalForce="0,0,0") -> (Pose) {
    angularDampingOverride=1
    angularSpringConstant=5000
    bAngularSpring=true
    bOverrideAngularDamping=true
    bUseGravityOverride=true
    gravityScale=0.35
    simulationSpace=RootRelative
    alphaBoolBlend {blendOption=Linear}
    boundBone {boneName=base_pad_leg_l}
    componentAppliedLinearAccClamp=<3000 3000 3000>
    componentLinearAccScale=<0.15 0.15 0.15>
    componentLinearVelScale=<0.15 0.15 0.15>
    gravityOverride {z=-980}
    physicsBodyDefinitions [{
      boundBone {boneName=base_pad_leg_l}
      boxExtents=<35 15 15>
      constraintSetup {angularLimitsMax=<5 5 10>, angularLimitsMin=<-5 -5 -10>, angularTarget=<1 0 0>}
      localJointOffset=<2 0 0>
     }]
    simSpaceSettings {maxAngularAcceleration=35, maxAngularVelocity=7, simSpaceAngularAlpha=0.2}
   }
   n3 = Anim Dynamics - Bone: base_pad_reg_r(ComponentPose=Local To Component(LocalPose=Use cached pose 'CtrlRig_Cache'()), Alpha=1, GravityOverride="0,0,0", ExternalForce="0,0,0") {
    angularDampingOverride=1
    angularSpringConstant=5000
    bAngularSpring=true
    bOverrideAngularDamping=true
    bUseGravityOverride=true
    gravityScale=0.35
    simulationSpace=RootRelative
    alphaBoolBlend {blendOption=Linear}
    boundBone {boneName=base_pad_reg_r}
    componentAppliedLinearAccClamp=<2000 2000 2000>
    componentLinearAccScale=<0.15 0.15 0.15>
    componentLinearVelScale=<0.15 0.15 0.15>
    gravityOverride {z=-980}
    physicsBodyDefinitions [{
      boundBone {boneName=base_pad_reg_r}
      boxExtents=<35 15 15>
      constraintSetup {angularLimitsMax=<5 5 10>, angularLimitsMin=<-5 -5 -10>, angularTarget=<1 0 0>}
      localJointOffset=<2 0 0>
     }]
    relativeSpaceBone {boneName=root}
   }
   n4 = Anim Dynamics - Bone: base_pad_leg_l(ComponentPose=@n3, Alpha=1, GravityOverride="0,0,0", ExternalForce="0,0,0") {
    angularDampingOverride=1
    angularSpringConstant=5000
    bAngularSpring=true
    bOverrideAngularDamping=true
    bUseGravityOverride=true
    gravityScale=0.35
    simulationSpace=RootRelative
    alphaBoolBlend {blendOption=Linear}
    boundBone {boneName=base_pad_leg_l}
    componentAppliedLinearAccClamp=<2000 2000 2000>
    componentLinearAccScale=<0.15 0.15 0.15>
    componentLinearVelScale=<0.15 0.15 0.15>
    gravityOverride {z=-980}
    physicsBodyDefinitions [{
      boundBone {boneName=base_pad_leg_l}
      boxExtents=<35 15 15>
      constraintSetup {angularLimitsMax=<5 5 10>, angularLimitsMin=<-5 -5 -10>, angularTarget=<1 0 0>}
      localJointOffset=<2 0 0>
     }]
    relativeSpaceBone {boneName=root}
   }
   Output Pose(Result=Blend Poses by bool(BlendPose_0=Component To Local(ComponentPose=@n2), BlendPose_1=Component To Local(ComponentPose=@n4), bActiveValue=Get bIsInFrontEnd())) {layerGroup=None}
   n5 = Control Rig!(
    Source=Copy Pose From Mesh() {bUseAttachedParent=true}
    Alpha
    parachuting:bool=Get parachuting() -> (parachuting:bool)
    upperarm_transform:Transform=Get upperarm_transform() -> (upperarm_transform:Transform)
    upperarm_transform_parachuting:Transform=Get upperarm_transform_parachuting() -> (upperarm_transform_parachuting:Transform)
    isEmotePlay:bool=Get bIsPlayingEmote()
    isFE:bool=Get bIsInFrontEnd()
   ) -> (Pose) {controlRigAssetReference {blueprintRigClass=&/M_MED_PixieParts_Magma_CtrlRig.M_MED_PixieParts_Magma_CtrlRig_C}}
   Save cached pose 'CtrlRig_Cache'(Pose=@n5) {cacheName=CtrlRig_Cache}
  ]}
 EventGraph {nodes [
   n1 = Cast To PlayerPawn_Athena(Object=TryGetPawnOwner())
   > Sequence()
   Event BlueprintUpdateAnimation()
   > Set parachuting(parachuting:bool=Get bIsParachuteOpened()) -> (Output_Get:bool)
   n2 = Event OnCharacterCustomizationLoaded()
   n3 = GetObjectName(Object=Get SkeletalMeshAsset!(self:Skeletal Mesh Component Object Reference=Get Array Item with EFort Custom Part Type(Array=@n2.CustomMeshParts, PartType=Hat)) -> (SkeletalMeshAsset:Skeletal Mesh Object Reference))
   > n4 = Branch(Condition=Contains(SearchIn=@n3, Substring=drill))
   > Set upperarm_transform(upperarm_transform:Transform=MakeTransform(Location="1,0,7", Rotation="0, 0, 0")) -> (Output_Get:Transform)
   > Set upperarm_transform_parachuting(upperarm_transform_parachuting:Transform=MakeTransform(Location="1,-7,0", Rotation="0, 0, 0")) -> (Output_Get:Transform)
   n5 = Branch(execute=@n4.else, Condition=Contains(SearchIn=@n3, Substring=bolt))
   > Set upperarm_transform(upperarm_transform=MakeTransform(Location="4,0,3", Rotation="0, 0, 0"))
   > Set upperarm_transform_parachuting(upperarm_transform_parachuting=MakeTransform(Location="0,-4,0", Rotation="0, 0, 0"))
   n6 = Branch(execute=@n5.else, Condition=Contains(SearchIn=@n3, Substring=screw))
   > Set upperarm_transform(upperarm_transform=MakeTransform(Location="1,0,2", Rotation="0, 0, 0"))
   > Set upperarm_transform_parachuting(upperarm_transform_parachuting=MakeTransform(Location="1,-3,0", Rotation="0, 0, 0"))
   Set upperarm_transform(execute=@n6.else, upperarm_transform=MakeTransform(Location="2.5,0,2", Rotation="0, 0, 0"))
   > Set upperarm_transform_parachuting(upperarm_transform_parachuting=MakeTransform(Location="2,-3,0", Rotation="0, 0, 0"))
   n7 = Set Body Mesh(Body Mesh:Skeletal Mesh Component Object Reference=GetSkeletalMeshForPartType(self=@n1, PartType=Face)) -> (Output_Get:Skeletal Mesh Component Object Reference)
   > AddTickPrerequisiteComponent!(execute, self:Actor Component Object Reference=GetOwningComponent(), PrerequisiteComponent:Actor Component Object Reference=@n7) -> (then)
  ]}
]
variables [
 Body Mesh {type=SkeletalMeshComponent}
 parachuting {type=boolean}
 upperarm_transform {type=Transform, default {location=<0 0 0>, rotation=<rot 0 0 0>, scale=<1 1 1>}}
 upperarm_transform_parachuting {type=Transform, default {location=<0 0 0>, rotation=<rot 0 0 0>, scale=<1 1 1>}}
]
